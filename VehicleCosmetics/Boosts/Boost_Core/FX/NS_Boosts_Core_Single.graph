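# Niagara system for the Core vehicle boost cosmetic in its single-boost form.
# A ribbon trail, a burst-spawned sprite-and-mesh core with sine wobble and radial material, and a rearward cone of stretched sparks form the exhaust.
# Ribbons and sparks inherit source movement to follow the car.
# Each emitter resolves layered painted-variant colors, while user parameters adjust offsets and brightness for frontend preview and scale ribbon alpha by sky color.
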NiagaraSystem /VehicleCosmetics/Boosts/Boost_Core/FX/NS_Boosts_Core_Single
dependencies {
 usedDataInterfaces [
  &NiagaraDataInterfaceCurve
  &NiagaraDataInterfaceArrayFloat
  &NiagaraDataInterfaceVector2DCurve
  &NiagaraDataInterfaceVectorCurve
  &NiagaraDataInterfaceSkeletalMesh
 ]
 usedDynamicInputs []
 usedModules [
  &self:SystemSpawnScript.SystemScriptSource.SystemScriptGraph.NiagaraNodeAssignment_0.SetVariables_#1
  &SystemState
 ]
}
emitters {
 NE_VC_SingleBoost_Ribbon_Parent {
  emitterUpdateScript [
   &EmitterState()
   &VC_FilteredSocketCount()
   &VC_RateQualityHandler()
   &VC_SpawnRibbon_v1(SpawnRate=$Emitter.Spawn Rate)
  ]
  particleSpawnScript [
   &VC_PaintedVariantsOverride(
    Black=<rgba 0 0 0 30>
    Cobalt=<rgba 0 0.0925 5 1>
    Crimson=<rgba 2 0.059 0.0399 1>
    DefaultColor=<rgba 1 0.2368 0.0573 1>
    ForestGreen=<rgba 0.06 1.2 0.01 1>
    Gold:LinearColor=<rgba 1 0.5826 0.3715 1>
    Grey=<rgba 0.0151 0.0168 0.0208 30>
    Lime=<rgba 0.5052 1 0.0104 1>
    Orange=<rgba 0.1406 0.0431 0 4>
    Purple=<rgba 0.7105 0.0419 5 1>
    R1 Black:LinearColor=<rgba 0 0 0 50>
    R1 Crimson:LinearColor=<rgba 0.7448 0.0078 0.0078 5>
    R1 DefaultColor=<rgba 1 0.2368 0.0573 1>
    R1 ForestGreen:LinearColor=<rgba 0.0261 1 0.01 1>
    R1 Grey:LinearColor=<rgba 0.02 0.02 0.02 50>
    R1 Lime:LinearColor=<rgba 0.4261 0.5885 0 1>
    R1 Orange:LinearColor=<rgba 0.1667 0.1062 0.0016 4>
    R1 Purple:LinearColor=<rgba 0.1939 0.0042 0.7708 4>
    R1 Saffron:LinearColor=<rgba 0.06 0.05 0 30>
    R1 SkyBlue:LinearColor=<rgba 0.005 0.1116 0.1406 1>
    R1 TitaniumWhite:LinearColor=<rgba 0.026 0.026 0.026 0>
    R2 Black:LinearColor=<rgba 0 0 0 15>
    R2 BurntSienna:LinearColor=<rgba 0 0 0 0>
    R2 Cobalt:LinearColor=<rgba 0 0 0 0>
    R2 Crimson:LinearColor=<rgba 0 0 0 0>
    R2 ForestGreen:LinearColor=<rgba 0 0 0 0>
    R2 Gold:LinearColor=<rgba 0 0 0 0>
    R2 Grey:LinearColor=<rgba 0 0 0 0>
    R2 Lime:LinearColor=<rgba 0 0 0 0>
    R2 Orange:LinearColor=<rgba 0 0 0 0>
    R2 Pink:LinearColor=<rgba 0 0 0 0>
    R2 Purple:LinearColor=<rgba 0 0 0 0>
    R2 Saffron:LinearColor=<rgba 0 0 0 0>
    R2 SkyBlue:LinearColor=<rgba 0 0 0 0>
    R2 TitaniumWhite:LinearColor=<rgba 0 0 0 0>
    Saffron=<rgba 1 0.7915 0 1>
    SkyBlue=<rgba 0 0.49 0.5365 1>
    TitaniumWhite=<rgba 5 5 5 0.02>
    Use R1 Painted Overrides=true
    Use R2 Painted Overrides=true
   )
   &VC_InitializeRibbon_v1(
    Color=$Particles.Color
    Custom Offset=true
    Lifetime=&MakeCustomFloatFromBool(Bool=$User.bIsFrontendPreview, True Float=0.6, False Float=1)
    Local Vector Offset=/Niagara/DynamicInputs/TypeConversions/MakeCustomPositionFromBool(Bool:NiagaraBool=$User.bIsFrontendPreview, True Position:NiagaraPosition=<-150 0 25>, False Position:NiagaraPosition=<-140 0 15>)
    Offset Amount=1
    Ribbon Width=150
   )
   &Color(Color=&MakeCustomLinearColorFromBool(Bool=$User.bIsPainted, True Color=$Particles.Color, False Color=<rgba 1 0.2368 0.0573 1>)) {enabled=false}
   &AddVelocity(Velocity=<-1300 0 0>, Velocity Speed Scale=2) {module=AddVelocity001}
  ]
  particleUpdateScript [
   &ParticleState()
   &InheritSourceMovement(
    Applied Position Scale=&VectorFromFloat(Value=&FloatFromCurve(FloatCurve=DataInterface() {propertyValues {Curve {keys [
          {interpMode=RCIM_Cubic, leaveTangent=-0.088, tangentMode=RCTM_Break, time=0, value=1}
          {arriveTangent=-1.6903, interpMode=RCIM_Cubic, tangentMode=RCTM_Break, time=1.0049, value=0.5795}
         ]}}}))
    ApplyForce=false
    ApplyPosition=true
    ApplyPositionFalloff=false
   )
   &SolveForcesAndVelocity(Clamp Velocity=false, Limit Acceleration=false)
   &VC_UpdateRibbon_v1(Enable Skeletal Position=true)
   &ScaleRibbonWidth(Ribbon Width Scale=&FloatFromCurve(FloatCurve=DataInterface() {propertyValues {Curve {keys [<at 0 1.4163>,<at 0.3 1.2471>,<at 0.6927 0.5324>,<at 1 0.2>]}}}))
   /SetVariables(Particles.Ribbon2Width:NiagaraFloat=&FloatFromCurve(
     FloatCurve=DataInterface() {propertyValues {Curve {keys [<at 0.0024 1.4164>,<at 0.2959 1.2188>,<at 0.8 0.65>,<at 1 0.5851>]}}}
     Scale Curve=&MakeCustomFloatFromBool(Bool=$User.bIsFrontendPreview, True Float=305.3998, False Float=300)
    )) {isSetParameters=true}
   /SetVariables(Particles.Color1:LinearColor=&self:VC_Dynamic_ScaleColor(
     Color to Scale:LinearColor=$Particles.Color
     RGB Scale:Vector3f=&VectorFromFloat(Value=&FloatFromCurve(
       FloatCurve=DataInterface() {propertyValues {Curve {keys [<at 0 20.25>,<at 0.2 13>,<at 0.5 7.5>,<at 0.7026 4.5824>,<at 1.0051 -0.4235>]}}}
       Scale Curve=2
      ))
     Alpha Scale:NiagaraFloat=&Multiply_Float(
      A=&FloatFromCurve(
       FloatCurve=DataInterface() {propertyValues {Curve {keys [<at 0 0>,<at 0.1 0.9>,<at 0.2 0.8>,<at 0.5 0.25>,<at 0.7052 0.1947>,<at 1 0>]}}}
       Scale Curve=&MakeCustomFloatFromBool(Bool=$User.bIsPainted, True Float=$User.MultiplierSkyColor, False Float=1)
      )
      B=&MakeCustomFloatFromBool(
       Bool=$User.bIsFrontendPreview
       True Float=&SelectFloatFromArray(
        Direct Array Index=$User.ColorVariantID
        Float Selection Array=DataInterface() {propertyValues {FloatData [0.45,0.45,0.45,0.45,0.45,0.45,0.55,0.45,0.2,0.45,0.45,0.45,0.15,0.45,0.15]}}
       )
       False Float=1
      )
     )
    )) {isSetParameters=true}
   /SetVariables(Particles.Color2=&self:VC_Dynamic_ScaleColor_01(
     Color to Scale:LinearColor=&MakeCustomLinearColorFromBool(Bool=$User.bIsPainted, True Color=$Particles.R1 Color, False Color=$Particles.R1 Color)
     RGB Scale:Vector3f=&VectorFromFloat(Value=&FloatFromCurve(FloatCurve=DataInterface() {propertyValues {Curve {keys [<at 0.0026 9.8588>,<at 1.0052 5.9059>]}}}))
     Alpha Scale:NiagaraFloat=&Multiply_Float(
      A=&FloatFromCurve(
       FloatCurve=DataInterface() {propertyValues {Curve {keys [<at 0 1.45>,<at 0.2 0.9441>,<at 0.5 0.2912>,<at 0.9948 0.0824>]}}}
       Scale Curve=&MakeCustomFloatFromBool(Bool=$User.bIsPainted, True Float=$User.MultiplierSkyColor, False Float=0.5)
      )
      B=&MakeCustomFloatFromBool(Bool=$User.bIsFrontendPreview, True Float=0.2, False Float=1)
     )
    )) {isSetParameters=true}
   /SetVariables(Particles.Color3:LinearColor=&self:VC_Dynamic_ScaleColor(
     Color to Scale:LinearColor=$Particles.R2 Color
     RGB Scale:Vector3f=<1 1 1>
     Alpha Scale:NiagaraFloat=&FloatFromCurve(FloatCurve=DataInterface() {propertyValues {Curve {keys [<at -0.007 -0.2545>,<at 0.0668 1.9322>,<at 0.2901 1.574>,<at 0.5202 1.6089>,<at 0.9931 0.0223>]}}})
    )) {isSetParameters=true}
  ]
  rendererClasses [&NiagaraRibbonRendererProperties]
  renderers [
   {rendererClass=&NiagaraRibbonRendererProperties}
   {rendererClass=&NiagaraRibbonRendererProperties, rendererIndex=1}
  ]
 }
 NE_VC_SingleBoost_Sprite_Parent {
  emitterUpdateScript [
   &EmitterState(
    Inactive Response=::NewEnumerator1
    Life Cycle Mode=::NewEnumerator1
    Loop Behavior=::NewEnumerator1
    Loop Duration Mode=::NewEnumerator1
    UseLoopDelay=false
   )
   &SpawnRate(SpawnRate=50, Use Spawn Probability=false) {enabled=false}
   &SpawnBurst_Instantaneous(Use Spawn Probability=false, UseLoopCountLimit=false)
  ]
  particleSpawnScript [
   &VC_PaintedVariantsOverride(
    Black=<rgba 0 0 0 3>
    Cobalt=<rgba 0.0469 0.2853 3 1>
    Crimson=<rgba 2 0.1042 0.1556 1>
    DefaultColor=<rgba 1 0.2506 0.0844 1>
    ForestGreen=<rgba 0 1 0.0505 1>
    Grey=<rgba 0.0781 0.0781 0.0781 20>
    Lime=<rgba 0.3618 0.9531 0.0631 1>
    Orange=<rgba 0.9219 0.177 0.0443 1>
    Pink=<rgba 1 0.1198 0.6799 1>
    Purple=<rgba 0.3809 0.0781 1 1>
    R1 Black:LinearColor=<rgba 1.5 1.5 1.5 1>
    R1 Cobalt:LinearColor=<rgba 0.0417 0.2006 2 1>
    R1 Crimson:LinearColor=<rgba 1.5833 0.1072 0.1369 1>
    R1 DefaultColor=<rgba 3 0.9487 0.4583 1>
    R1 ForestGreen:LinearColor=<rgba 0.0312 1 0.0802 1>
    R1 Grey:LinearColor=<rgba 0.9792 0.9792 0.9792 1>
    R1 Lime:LinearColor=<rgba 0.3042 0.8021 0.142 1>
    R1 Orange:LinearColor=<rgba 1 0.2332 0.0156 1>
    R1 Pink:LinearColor=<rgba 2 0.2396 1.3598 1>
    R1 Purple:LinearColor=<rgba 0.7447 0.1667 2 1>
    R1 Saffron:LinearColor=<rgba 1 0.7813 0.0625 1>
    R1 SkyBlue:LinearColor=<rgba 0.3542 0.8191 2 1>
    R1 TitaniumWhite:LinearColor=<rgba 1 1 1 1>
    R2 Black:LinearColor=<rgba 1 1 1 0.8957>
    R2 BurntSienna:LinearColor=<rgba 0 0 0 0.8957>
    R2 Cobalt:LinearColor=<rgba 0 0 0 0.8957>
    R2 Crimson:LinearColor=<rgba 0 0 0 0.8957>
    R2 DefaultColor:LinearColor=<rgba 0 0 0 0.8957>
    R2 ForestGreen:LinearColor=<rgba 0 0 0 3>
    R2 Gold:LinearColor=<rgba 0 0 0 0.8957>
    R2 Grey:LinearColor=<rgba 0 0 0 0.8957>
    R2 Lime:LinearColor=<rgba 0 0 0 0.99>
    R2 Orange:LinearColor=<rgba 0 0 0 0.8957>
    R2 Pink:LinearColor=<rgba 0 0 0 0.8957>
    R2 Purple:LinearColor=<rgba 0 0 0 0.8957>
    R2 Saffron:LinearColor=<rgba 0 0 0 0.97>
    R2 SkyBlue:LinearColor=<rgba 0 0 0 0.8957>
    R2 TitaniumWhite:LinearColor=<rgba 0.0417 0.0417 0.0417 1>
    Saffron=<rgba 0.9 0.6998 0.0422 1>
    SkyBlue=<rgba 0.0104 0.5849 1 3>
    TitaniumWhite=<rgba 0.6146 0.6146 0.6146 1>
    Use R1 Painted Overrides=true
    Use R2 Painted Overrides=true
   )
   &InitializeParticle(
    Color=$Particles.Color
    Lifetime=$Emitter.CurrentLoopDuration
    Mesh Scale Mode=::NewEnumerator3
    Mesh Uniform Scale=0.5
    Position Offset=&AddVectorToPosition(Position=<-230 0 25>, Vector=Linked() {linkedVariable {name=User.SingleBoostPos, type=NiagaraPosition}})
    Position Offset Coordinate Space=::Local
    Sprite Rotation Mode=::NewEnumerator2
    Sprite Size Mode=::NewEnumerator3
    Uniform Sprite Size=190
    UsePositionOffset=true
    Write Material Random=true
   )
   &Color(Color=$Particles.Color)
   /SetVariables(Particles.ColorRays:LinearColor=&Multiply_LinearColorByFloat(
     Linear Color=$Particles.R1 Color
     Float=&MakeCustomFloatFromBool(Bool=$User.bIsFrontendPreview, True Float=1, False Float=2)
    )) {isSetParameters=true}
   /SetVariables(Particles.Color2=$Particles.R2 Color) {isSetParameters=true}
   /SetVariables(Particles.Scale2:Vector3f=&Multiply_VectorByFloat(Vector=$Particles.Scale, Float=1.05)) {isSetParameters=true}
   &AddVelocity(Velocity=<-1500 0 0>) {enabled=false}
  ]
  particleUpdateScript [
   &ParticleState()
   &ShapeLocation(
    MaskSpawnGroup=false
    Offset=&MakeVector(
     X=&Sine(Normalized Angle=$Emitter.Age, Period=0.4, Scale=1.2, Bias=0)
     Y=&Sine(Normalized Angle=$Emitter.Age, Period=0.65, Scale=1, Bias=0.3)
     Z=&Sine(Normalized Angle=$Emitter.Age, Period=1, Scale=1.3, Bias=0.4014)
    )
    Offset Mode=::NewEnumerator0
    Override Seed=false
    Shape Origin=$Particles.Initial.Position
    Sphere Radius=0
    UseOwnerScale=true
   )
   &ScaleMeshSize(Scale Factor=&VectorFromFloat(Value=&FloatFromCurve(
      FloatCurve=DataInterface() {propertyValues {Curve {keys [{tangentMode=RCTM_User, time=0, value=0}, <at 0.1 0.2>, {tangentMode=RCTM_User, time=0.35, value=1}]}}}
      CurveIndex=$Emitter.Age
     )))
   &FN_Radial_Material(
    Base Sprite Offset X=0.296
    Base Sprite Tile U=0.06
    Depth Fade=16.3904
    Radial Mask Offset X=0.06
    Radial Mask Power=2.5286
    Radial Mask Strength=2.6
    Radial Mask Tile V=1.0168
    Sprite Extrusion Speed=15
    TIme:NiagaraFloat=$System.NormalizedLoopAge
    Write to Module Namespace=true
   )
   &SolveForcesAndVelocity(Clamp Velocity=false, Limit Acceleration=false)
   &ScaleColor(Scale RGB=&Multiply_VectorByFloat(Vector=<1 1 1>, Float=1), ScaleA=true, ScaleRGB=true)
  ]
  rendererClasses [&NiagaraSpriteRendererProperties, &NiagaraMeshRendererProperties]
  renderers [
   {rendererClass=&NiagaraSpriteRendererProperties}
   {rendererClass=&NiagaraMeshRendererProperties, rendererIndex=1}
   {rendererClass=&NiagaraMeshRendererProperties, rendererIndex=2}
  ]
 }
 NE_VC_SingleBoost_Sprite_Parent001 {
  emitterUpdateScript [
   &EmitterState()
   &SpawnRate(SpawnRate=60, Use Spawn Probability=false)
  ]
  particleSpawnScript [
   &VC_PaintedVariantsOverride(
    Black=<rgba 3 3 3 1>
    Cobalt=<rgba 0.1346 0.9114 5 1>
    Crimson=<rgba 5 0.3125 0.4395 1>
    DefaultColor=<rgba 25 4.8535 0.2892 1>
    ForestGreen=<rgba 0 0.9062 0.0458 1>
    Grey=<rgba 1 1 1 3>
    Orange=<rgba 10 1.3413 0.0049 1>
    Purple=<rgba 1.6448 0.018 5 1>
    R1 Black:LinearColor=<rgba 0.9531 0.9531 0.9531 0.744>
    R1 Cobalt:LinearColor=<rgba 0 0.032 1 1>
    R1 DefaultColor=<rgba 1 0.1941 0.0116 1>
    R1 ForestGreen:LinearColor=<rgba 0 0.6719 0.0339 1>
    R1 Grey:LinearColor=<rgba 0.0833 0.0833 0.0833 15>
    R1 Orange:LinearColor=<rgba 0.2292 0.0799 0 1>
    R1 Saffron:LinearColor=<rgba 0.276 0.2116 0 1>
    Saffron=<rgba 1 0.7667 0 1>
    SkyBlue=<rgba 0 4.6577 7 1>
    Use R1 Painted Overrides=true
   )
   &InitializeParticle(
    Color=$Particles.Color
    Lifetime=&RandomRangeFloat(Minimum=0.2, Maximum=0.4)
    Position Offset=&AddVectorToPosition(
     Position=&MakeCustomPositionFromBool(Bool:NiagaraBool=$User.bIsFrontendPreview, True Position:NiagaraPosition=<-260 0 20>, False Position:NiagaraPosition=<-200 0 15>)
     Vector=Linked() {linkedVariable {name=User.SingleBoostPos, type=NiagaraPosition}}
    )
    Position Offset Coordinate Space=::Local
    Sprite Rotation Mode=::NewEnumerator2
    Sprite Size Mode=::NewEnumerator1
    Uniform Sprite Size Max=5
    Uniform Sprite Size Min=1
    UsePositionOffset=true
    Write Material Random=true
   )
   &ShapeLocation(
    MaskSpawnGroup=false
    Override Seed=false
    Sphere Radius=&MakeCustomFloatFromBool(Bool=$User.bIsFrontendPreview, True Float=50, False Float=60)
    UseOwnerScale=true
   )
   &AddVelocity(
    Cone Angle=5
    Distribution Along Cone Axis=0.8
    Module.UseConeFalloff=false
    Module.UseVelDistribution=true
    Override Seed=false
    Velocity Mode=::NewEnumerator2
    Velocity Speed=&RandomRangeFloat(Minimum=-1500, Maximum=-2200)
   )
  ]
  particleUpdateScript [
   &ParticleState()
   &ScaleSpriteSize(
    Non-Uniform Curve Scale=&Multiply_Vector2D_ByFloat(
     Float=1
     Vector2D=&MakeCustomVector2DFromBool(Bool=$User.bIsFrontendPreview, True Vector2D=<1 5>, False Vector2D=<1 2>)
    )
    Non-Uniform Curve Sprite Scale=DataInterface() {propertyValues {
      XCurve {keys [{tangentMode=RCTM_User, time=0, value=1}, <at 0.5 1>, {tangentMode=RCTM_User, time=1, value=0}]}
      YCurve {keys [{tangentMode=RCTM_User, time=0, value=11}, <at 0.2 9>, <at 0.5 3>, {tangentMode=RCTM_User, time=1, value=0}]}
     }}
    Scale Sprite Size Mode=::NewEnumerator3
   )
   /SetVariables(Particles.Sprite2Size:Vector2f=&Multiply_Vector2D(A=$Particles.SpriteSize, B=<6 1.8385>)) {isSetParameters=true}
   /SetVariables(Particles.Color1:LinearColor=&Multiply_LinearColorByFloat(
     Linear Color=$Particles.Color
     Float=&MakeCustomFloatFromBool(Bool=$User.bIsFrontendPreview, True Float=1, False Float=7)
    )) {isSetParameters=true}
   /SetVariables(Particles.Color2=&Multiply_LinearColorByFloat(
     Linear Color=$Particles.R1 Color
     Float=&MakeCustomFloatFromBool(Bool=$User.bIsFrontendPreview, True Float=0.464, False Float=1)
    )) {isSetParameters=true}
   &ScaleVelocity(
    Coordinate Space:ENiagaraCoordinateSpace=::World
    Velocity Scale=&VectorFromCurve(VectorCurve=DataInterface() {propertyValues {XCurve {keys [<at 0 0.05>,<at 0.3 1>]}, YCurve {keys [<at 0 0.05>,<at 0.3 1>]}, ZCurve {keys [<at 0 0.05>,<at 0.3 1>]}}})
   )
   &InheritSourceMovement(
    Applied Force Scale=<1 1.2 1>
    Applied Position Scale=&Multiply_VectorByFloat(
     Vector=<1 1 1>
     Float=&FloatFromCurve(FloatCurve=DataInterface() {propertyValues {Curve {keys [
          {arriveTangent=-0.0449, interpMode=RCIM_Cubic, leaveTangent=-0.0449, tangentMode=RCTM_User, time=0, value=1}
          {arriveTangent=-1.5389, interpMode=RCIM_Cubic, leaveTangent=-1.5389, tangentMode=RCTM_User, time=1, value=0}
         ]}}})
    )
    ApplyForce=true
    ApplyForceFalloff=false
    ApplyPosition=true
    ApplyPositionFalloff=false
    Module.LimitForceSpeed=false
   )
   &SolveForcesAndVelocity(Clamp Velocity=false, Limit Acceleration=false)
  ]
  rendererClasses [&NiagaraSpriteRendererProperties]
  renderers [
   {rendererClass=&NiagaraSpriteRendererProperties}
   {rendererClass=&NiagaraSpriteRendererProperties, rendererIndex=1}
  ]
 }
}
summary {
 systemName=NS_Boosts_Core_Single
 emitters [
  {emitterName=NE_VC_SingleBoost_Ribbon_Parent, rendererClasses [&NiagaraRibbonRendererProperties]}
  {
   emitterName=NE_VC_SingleBoost_Sprite_Parent
   rendererClasses [&NiagaraSpriteRendererProperties, &NiagaraMeshRendererProperties]
  }
  {emitterName=NE_VC_SingleBoost_Sprite_Parent001, rendererClasses [&NiagaraSpriteRendererProperties]}
 ]
 userVariables [
  User.bIsFrontendPreview {type=NiagaraBool}
  User.bIsPainted {type=NiagaraBool}
  User.ColorVariantID {type=NiagaraInt32}
  User.DTCarOffset {type=Vector3f, description="", defaultValue {struct=Vector3f, value=<0 0 0>}}
  User.GameModeID {type=NiagaraInt32}
  User.MultiplierSkyColor {type=NiagaraFloat, defaultValue=1, description=""}
  User.PaintedColor {type=LinearColor, description="", defaultValue {struct=LinearColor, value=<rgba 1 1 1 1>}}
  User.PaintedScalar {type=NiagaraFloat, defaultValue=1, description=""}
  User.PortCount {type=NiagaraInt32, defaultValue=1}
  User.SingleBoostPos {type=NiagaraPosition, description="", defaultValue {}}
 ]
}
