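# Translucent, two-sided Niagara material in the vehicle boost cosmetics core, usable on mesh, sprite and ribbon particles for scrolling noise effects.
# Emissive blends particle color with an optional panned texture through an adjustment function.
# Opacity comes from a panned cell-noise mask with distorted UVs, inverted, thresholded and sharpened, then added to or multiplied by Fresnel.
# Particle alpha scales the result, with depth fade skipped on mobile.
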Material /VehicleCosmetics/Boosts/Boost_Core/Materials/M_Particle_Mesh_Noise_Distortion1
graph {
 properties {
  bUsedWithNiagaraMeshParticles=true
  bUsedWithNiagaraRibbons=true
  bUsedWithNiagaraSprites=true
  blendMode=BLEND_Translucent
  pixelDepthOffsetMode=PDOM_Legacy
  twoSided=true
  usedShadingModels=Default Lit
 }
 nodes [
  ComponentMask() {g=true, r=true}
  n1 = ParticleColor()
  n2 = Multiply(
   A=TextureCoordinate()
   B=AppendVector(
    A=ScalarParameter() {defaultValue=1, group="1) Emissive", parameterName=Emissive-Tiling_X}
    B=ScalarParameter() {defaultValue=1, group="1) Emissive", parameterName=Emissive-Tiling_Y}
   )
  )
  n3 = AppendVector(
   A=ScalarParameter() {group="1) Emissive", parameterName=Emissive-Speed_X}
   B=ScalarParameter() {group="1) Emissive", parameterName=Emissive-Speed_Y}
  )
  n4 = ChannelMaskParameter(None=TextureSampleParameter2D(UVs=Panner(Coordinate=@n2, Speed=@n3)) {group="1) Emissive", parameterName=Emissive_Texture, texture=/Game/Global/Textures/Default/Blanks/T_WhiteUserInterface}.RGB) {group="1) Emissive", parameterName=Emissive-Texture_ChannelPicker}
  MaterialFunctionCall(Emissive=LinearInterpolate(
    A=@n1.RGB
    B=Multiply(A=@n4.R, B=@n1.RGB)
    Alpha=ScalarParameter() {group="1) Emissive", parameterName=Use_Emissive_Texture, sortPriority=-1}
   )) {materialFunction=&MF_FX_EmissiveAdjustments}
  n5 = Panner(Coordinate=TextureCoordinate() {uTiling=0.5, vTiling=0.5}) {speedY=1}
  n6 = ScalarParameter() {group=Distortion, parameterName=Distortion Strength, sortPriority=11}
  Multiply(
   A=ComponentMask(None=TextureSampleParameter2D(UVs=@n5) {parameterName=distortion text, texture=/DealershipSharedLibrary/FX/Textures/Noise/T_VCVFX_Flow03_N}.RGB) {g=true, r=true}
   B=@n6
  )
  n7 = MaterialFunctionCall(
   Brightness=ScalarParameter() {defaultValue=1, group=Fresnel, parameterName=Fresnel_Brightness}
   Exponent=ScalarParameter() {defaultValue=20, group=Fresnel, parameterName=Fresnel_Exp}
   Vertex / Pixel Normal=VertexNormalWS()
  ) {materialFunction=&MF_Fresnel_FX}
  n8 = ScalarParameter() {defaultValue=1, group=Fresnel, parameterName=Double Sided Fresnel}
  n9 = ScalarParameter() {defaultValue=1, group=Fresnel, parameterName=Double Sided Fresnel}
  n10 = Multiply(
   A=LinearInterpolate(
    A=LinearInterpolate(A=@n7.Fresnel, B=@n7.Fresnel, Alpha=@n8)
    B=LinearInterpolate(A=@n7.Fresnel Inverted, B=@n7.Fresnel Inverted, Alpha=@n9)
    Alpha=ScalarParameter() {defaultValue=1, group=Fresnel, parameterName=InvertFresnel}
   )
   B=ScalarParameter() {defaultValue=0.1, group=Fresnel, parameterName=Fresnel_Multiplier}
  )
  n11 = VectorParameter() {group=Distortion, parameterName=Tiling And Speed Distortion, sortPriority=2, channelNames {a=Speed Y, b=Speed X, g=Tile Y, r=Tile X}}
  n12 = LinearInterpolate(
   B=MaterialFunctionCall!(
    Texture=TextureObjectParameter() {parameterName=Distortion Texture}
    Tiling=MaterialFunctionCall(X=@n11.Tile X, Y=@n11.Tile X) {materialFunction=&MakeFloat2}.Result
    Panner Time=ScalarParameter() {defaultValue=1, group=Distortion, parameterName=Panner Time Distortion, sortPriority=10}
    Offset XY
    Panner Speed=MaterialFunctionCall(X=@n11.Speed X, Y=@n11.Speed X) {materialFunction=&MakeFloat2}.Result
    Distortion Strength=@n6
    Channel Selector=VectorParameter() {group=Distortion, parameterName=Distortion Channel, sortPriority=1, defaultValue=<rgba 1 0 0 0>}.RGB
   ) {materialFunction=/DealershipSharedLibrary/FX/Materials/MaterialFunctions/MF_TextureDistortion}.Distortion
   Alpha=ScalarParameter() {defaultValue=1, parameterName=Use Distortion}
  )
  n13 = ScalarParameter() {defaultValue=1, group=BaseOpacityMask, parameterName=OpacityBase-Tiling_X}
  n14 = ScalarParameter() {defaultValue=1, group=BaseOpacityMask, parameterName=OpacityBase-Tiling_Y}
  n15 = Panner(
   Coordinate=Multiply(A=TextureCoordinate(), B=AppendVector(A=@n13, B=@n14))
   Speed=AppendVector(
    A=ScalarParameter() {group=BaseOpacityMask, parameterName=OpacityBase-Speed_X}
    B=ScalarParameter() {group=BaseOpacityMask, parameterName=OpacityBase-Speed_Y}
   )
  )
  n16 = ChannelMaskParameter(None=TextureSampleParameter2D(UVs=Add(A=@n12, B=@n15) {constB=0}) {group=BaseOpacityMask, parameterName=BaseOpacityMask, texture=/DealershipSharedLibrary/FX/Textures/Noise/T_VCVFX_CellNoise01_A}.RGB) {group=BaseOpacityMask, parameterName=Base_Opacity_ChannelPicker}
  n17 = LinearInterpolate(A=@n16.R, B=OneMinus(None=@n16.R), Alpha=ScalarParameter() {defaultValue=1, parameterName=Invert Opacity})
  n18 = ScalarParameter() {defaultValue=0.8845, parameterName=BaseOpacitySubtract}
  n19 = Clamp(None=Multiply(A=Subtract(A=@n17, B=@n18) {constB=0.15}, B=ScalarParameter() {defaultValue=15, parameterName=MultiplyBaseOpacity}) {constB=5})
  n20 = Multiply(
   A=@n1.A
   B=LinearInterpolate(
    A=Add(A=@n10, B=@n19)
    B=Multiply(A=@n10, B=@n19)
    Alpha=ScalarParameter() {group=Fresnel, parameterName=Add-Or-Multiply_Fresnel, sortPriority=-1}
   )
  )
  ShadingPathSwitch(
   Default=DepthFade(Opacity=@n20, FadeDistance=ScalarParameter() {defaultValue=10, group=BaseOpacityMask, parameterName=DepthFade_Distance, sortPriority=-1})
   Mobile=@n20
  )
 ]
}
